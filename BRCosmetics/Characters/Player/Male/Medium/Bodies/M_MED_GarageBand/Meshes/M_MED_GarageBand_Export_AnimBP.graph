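# Post-body AnimBlueprint for the M_MED_GarageBand male outfit that animates its seven dangling zipper bones on clavicles, chest, pockets and sleeves.
# It copies the attached parent mesh's pose, then weighted multi-target Constraint nodes place each zipper bone relative to nearby spine, pelvis and arm bones.
# Per-bone AnimDynamics then add damped, angle-limited secondary swing before output.
# There are no variables, and the update event is empty.
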AnimBlueprint /BRCosmetics/Characters/Player/Male/Medium/Bodies/M_MED_GarageBand/Meshes/M_MED_GarageBand_Export_AnimBP
target_skeleton=&/M_MED_GarageBand_Export_Skeleton
graphs [
 AnimGraph {nodes [
   n1 = "Constraint - dyn_zipper_clav_l (PPPPPP)"!(
    ComponentPose=Local To Component(LocalPose=Copy Pose From Mesh() {bUseAttachedParent=true})
    Alpha=1
    ConstraintWeights_0=0.7489
    ConstraintWeights_1:float=0.0339
    ConstraintWeights_2:float=0.0479
    ConstraintWeights_3:float=0.1289
    ConstraintWeights_4:float=0.0077
    ConstraintWeights_5:float=0.0327
   ) -> (Pose) {
    alphaBoolBlend {blendOption=Linear}
    boneToModify {boneName=dyn_zipper_clav_l}
    constraintSetup [
     {targetBone {boneName=clavicle_l}}
     {targetBone {boneName=spine_02}}
     {targetBone {boneName=spine_03}}
     {targetBone {boneName=spine_04}}
     {targetBone {boneName=spine_05}}
     {targetBone {boneName=upperarm_twist_01_l}}
    ]
    constraintWeights [1,1,1,1,1,1]
   }
   n2 = "Constraint - dyn_zipper_clav_r (PPPPP)"!(
    ComponentPose=@n1
    Alpha=1
    ConstraintWeights_0=0.0547
    ConstraintWeights_1:float=0.5518
    ConstraintWeights_2:float=0.0895
    ConstraintWeights_3:float=0.1805
    ConstraintWeights_4:float=0.1236
   ) -> (Pose) {
    alphaBoolBlend {blendOption=Linear}
    boneToModify {boneName=dyn_zipper_clav_r}
    constraintSetup [
     {targetBone {boneName=upperarm_twist_01_r}}
     {targetBone {boneName=clavicle_r}}
     {targetBone {boneName=spine_02}}
     {targetBone {boneName=spine_03}}
     {targetBone {boneName=spine_04}}
    ]
    constraintWeights [1,1,1,1,1]
   }
   n3 = "Constraint - dyn_zipper_chest (PPPPP)"!(
    ComponentPose=@n2
    Alpha=1
    ConstraintWeights_0=0.0001
    ConstraintWeights_1:float=0.3713
    ConstraintWeights_2:float=0.3116
    ConstraintWeights_3:float=0.2111
    ConstraintWeights_4:float=0.1059
   ) -> (Pose) {
    alphaBoolBlend {blendOption=Linear}
    boneToModify {boneName=dyn_zipper_chest}
    constraintSetup [
     {targetBone {boneName=pelvis}}
     {targetBone {boneName=spine_02}}
     {targetBone {boneName=spine_03}}
     {targetBone {boneName=spine_04}}
     {targetBone {boneName=spine_01}}
    ]
    constraintWeights [1,1,1,1,1]
   }
   n4 = "Constraint - dyn_zipper_pocket_l (PPPP)"!(
    ComponentPose=@n3
    Alpha=1
    ConstraintWeights_0=0.2513
    ConstraintWeights_1:float=0.2333
    ConstraintWeights_2:float=0.181
    ConstraintWeights_3:float=0.3344
   ) -> (Pose) {
    alphaBoolBlend {blendOption=Linear}
    boneToModify {boneName=dyn_zipper_pocket_l}
    constraintSetup [{targetBone {boneName=pelvis}}, {targetBone {boneName=spine_02}}, {targetBone {boneName=spine_03}}, {targetBone {boneName=spine_01}}]
    constraintWeights [1,1,1,1]
   }
   n5 = "Constraint - dyn_zipper_pocket_r (PPPP)"!(
    ComponentPose=@n4
    Alpha=1
    ConstraintWeights_0=0.2308
    ConstraintWeights_1:float=0.2446
    ConstraintWeights_2:float=0.1939
    ConstraintWeights_3:float=0.3307
   ) -> (Pose) {
    alphaBoolBlend {blendOption=Linear}
    boneToModify {boneName=dyn_zipper_pocket_r}
    constraintSetup [{targetBone {boneName=pelvis}}, {targetBone {boneName=spine_02}}, {targetBone {boneName=spine_03}}, {targetBone {boneName=spine_01}}]
    constraintWeights [1,1,1,1]
   }
   n6 = "Constraint - dyn_zipper_sleeve_r (PPPP)"!(
    ComponentPose=@n5
    Alpha=1
    ConstraintWeights_0=0.5911
    ConstraintWeights_1:float=0.0079
    ConstraintWeights_2:float=0.2245
    ConstraintWeights_3:float=0.1764
   ) -> (Pose) {
    alphaBoolBlend {blendOption=Linear}
    boneToModify {boneName=dyn_zipper_sleeve_r}
    constraintSetup [
     {targetBone {boneName=lowerarm_twist_02_r}}
     {targetBone {boneName=upperarm_r}}
     {targetBone {boneName=lowerarm_r}}
     {targetBone {boneName=lowerarm_twist_01_r}}
    ]
    constraintWeights [1,1,1,1]
   }
   n7 = "Constraint - dyn_zipper_sleeve_l (PPPP)"!(
    ComponentPose=@n6
    Alpha=1
    ConstraintWeights_0=0.2179
    ConstraintWeights_1:float=0.1828
    ConstraintWeights_2:float=0.0062
    ConstraintWeights_3:float=0.5931
   ) -> (Pose) {
    alphaBoolBlend {blendOption=Linear}
    boneToModify {boneName=dyn_zipper_sleeve_l}
    constraintSetup [
     {targetBone {boneName=lowerarm_l}}
     {targetBone {boneName=lowerarm_twist_01_l}}
     {targetBone {boneName=upperarm_l}}
     {targetBone {boneName=lowerarm_twist_02_l}}
    ]
    constraintWeights [1,1,1,1]
   }
   n8 = Anim Dynamics - Bone: dyn_zipper_clav_l(ComponentPose=@n7, Alpha=1, ExternalForce="0,0,0") {
    angularDampingOverride=0.98
    bGravityOverrideInSimSpace=true
    bOverrideAngularDamping=true
    alphaBoolBlend {blendOption=Linear}
    boundBone {boneName=dyn_zipper_clav_l}
    physicsBodyDefinitions [{
      sphereCollisionRadius=0
      boundBone {boneName=dyn_zipper_clav_l}
      boxExtents=<12 15 4>
      constraintSetup {twistAxis=AxisZ, angularLimitsMax=<45 5 5>, angularLimitsMin=<-45 -45 -5>, angularTarget=<0 0 0>}
      localJointOffset=<0 0 -2>
     }]
   }
   n9 = Anim Dynamics - Bone: dyn_zipper_clav_r(ComponentPose=@n8, Alpha=1, ExternalForce="0,0,0") {
    angularDampingOverride=0.95
    bGravityOverrideInSimSpace=true
    bOverrideAngularDamping=true
    alphaBoolBlend {blendOption=Linear}
    boundBone {boneName=dyn_zipper_clav_r}
    physicsBodyDefinitions [{
      sphereCollisionRadius=0
      boundBone {boneName=dyn_zipper_clav_r}
      boxExtents=<10 10 5>
      constraintSetup {twistAxis=AxisZ, angularLimitsMax=<45 5 5>, angularLimitsMin=<-45 -45 -5>, angularTarget=<0 0 0>}
      localJointOffset=<0 0 -2.2>
     }]
   }
   n10 = Anim Dynamics - Bone: dyn_zipper_chest(ComponentPose=@n9, Alpha=1, ExternalForce="0,0,0") {
    angularDampingOverride=0.95
    bGravityOverrideInSimSpace=true
    bOverrideAngularDamping=true
    alphaBoolBlend {blendOption=Linear}
    boundBone {boneName=dyn_zipper_chest}
    physicsBodyDefinitions [{
      sphereCollisionRadius=0
      boundBone {boneName=dyn_zipper_chest}
      boxExtents=<18 20 3.5>
      constraintSetup {twistAxis=AxisZ, angularLimitsMax=<45 5 5>, angularLimitsMin=<-45 -45 -5>, angularTarget=<0 0 0>}
      localJointOffset=<0 0 -1.5>
     }]
   }
   n11 = Anim Dynamics - Bone: dyn_zipper_sleeve_l(ComponentPose=@n10, Alpha=1, ExternalForce="0,0,0") {
    angularDampingOverride=0.97
    bGravityOverrideInSimSpace=true
    bOverrideAngularDamping=true
    alphaBoolBlend {blendOption=Linear}
    boundBone {boneName=dyn_zipper_sleeve_l}
    physicsBodyDefinitions [{
      sphereCollisionRadius=0
      boundBone {boneName=dyn_zipper_sleeve_l}
      boxExtents=<18 20 6>
      constraintSetup {twistAxis=AxisZ, angularLimitsMax=<45 5 5>, angularLimitsMin=<-45 -45 -5>, angularTarget=<0 0 0>}
      localJointOffset=<0 0 -2>
     }]
   }
   n12 = Anim Dynamics - Bone: dyn_zipper_sleeve_r(ComponentPose=@n11, Alpha=1, ExternalForce="0,0,0") {
    angularDampingOverride=0.95
    bGravityOverrideInSimSpace=true
    bOverrideAngularDamping=true
    alphaBoolBlend {blendOption=Linear}
    boundBone {boneName=dyn_zipper_sleeve_r}
    physicsBodyDefinitions [{
      sphereCollisionRadius=0
      boundBone {boneName=dyn_zipper_sleeve_r}
      boxExtents=<12 15 4>
      constraintSetup {twistAxis=AxisZ, angularLimitsMax=<120 5 5>, angularLimitsMin=<-120 -80 -5>, angularTarget=<0 0 0>}
      localJointOffset=<0 0 -2.5>
     }]
   }
   n13 = Anim Dynamics - Bone: dyn_zipper_pocket_l(ComponentPose=@n12, Alpha=1, ExternalForce="0,0,0") {
    angularDampingOverride=0.95
    bGravityOverrideInSimSpace=true
    bOverrideAngularDamping=true
    alphaBoolBlend {blendOption=Linear}
    boundBone {boneName=dyn_zipper_pocket_l}
    physicsBodyDefinitions [{
      sphereCollisionRadius=0
      boundBone {boneName=dyn_zipper_pocket_l}
      boxExtents=<12 15 8>
      constraintSetup {twistAxis=AxisZ, angularLimitsMax=<45 5 5>, angularLimitsMin=<-45 -45 -5>, angularTarget=<0 0 0>}
      localJointOffset=<0 0 -3>
     }]
   }
   n14 = Anim Dynamics - Bone: dyn_zipper_pocket_r(ComponentPose=@n13, Alpha=1, ExternalForce="0,0,0") {
    angularDampingOverride=0.96
    bGravityOverrideInSimSpace=true
    bOverrideAngularDamping=true
    alphaBoolBlend {blendOption=Linear}
    boundBone {boneName=dyn_zipper_pocket_r}
    physicsBodyDefinitions [{
      sphereCollisionRadius=0
      boundBone {boneName=dyn_zipper_pocket_r}
      boxExtents=<18 20 6>
      constraintSetup {twistAxis=AxisZ, angularLimitsMax=<45 5 5>, angularLimitsMin=<-45 -45 -5>, angularTarget=<0 0 0>}
      localJointOffset=<0 0 -3>
     }]
   }
   Output Pose(Result=Component To Local(ComponentPose=@n14)) {layerGroup=None}
  ]}
 EventGraph {nodes [Event BlueprintUpdateAnimation()]}
]
variables []
